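# Physics asset for the Neon Cat Fashion backpack cosmetic, defining bodies and joint limits for its ears, tail and zipper.
# Small capsules cover the Dyn_ bones, while the pack body adds a large box.
# Ears hinge mostly on one swing axis from Dyn_BackPack, which has no constraint to pack; the zipper and tail master constrain to pack.
# The tail chains from master through Tail_1, Tail_3, then Tail_2, out of numeric order.
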PhysicsAsset /BRCosmetics/Accessories/FORT_Backpacks/Backpack_F_MED_Neon_Cat_Fashion/Meshes/F_MED_Neon_Cat_Fashion_Pack_Physics
bodies {
 Dyn_BackPack [{shapeType=Capsule, length=0.505, radius=0.505, rotation=<rot 0 -90 0>}]
 Dyn_Ear_l [{shapeType=Capsule, length=2, radius=1, center=<1 1.0982 0>}]
 Dyn_Ear_r [{shapeType=Capsule, length=2, radius=1, center=<1 1.0982 0>}]
 Dyn_Tail_1 [{shapeType=Capsule, length=1, radius=1, center=<0.5 0 0>}]
 Dyn_Tail_2 [{shapeType=Capsule, length=1, radius=1, center=<0.5 0 0>}]
 Dyn_Tail_3 [{shapeType=Capsule, length=1, radius=1}]
 Dyn_Tail_Master [{shapeType=Capsule, length=0.505, radius=0.505, rotation=<rot 0 -90 0>}]
 Dyn_Zipper [{shapeType=Capsule, length=1, radius=0.5, center=<1 0 0>, rotation=<rot 90 0 0>}]
 pack [{shapeType=Capsule, length=1, radius=1}, {shapeType=Box, extentX=27.5, extentY=2, extentZ=35, center=<0 1.5829 -8.5409>}]
}
constraints [
 {bone1Name=Dyn_Ear_r, bone2Name=Dyn_BackPack, swing1LimitDegrees=30, swing2LimitDegrees=1, twistLimitDegrees=1}
 {bone1Name=Dyn_Ear_l, bone2Name=Dyn_BackPack, swing1LimitDegrees=30, swing2LimitDegrees=1, twistLimitDegrees=1}
 {bone1Name=Dyn_Zipper, bone2Name=pack, swing1LimitDegrees=45, swing2LimitDegrees=45, twistLimitDegrees=45}
 {bone1Name=Dyn_Tail_Master, bone2Name=pack, swing1LimitDegrees=45, swing2LimitDegrees=45, twistLimitDegrees=45}
 {bone1Name=Dyn_Tail_1, bone2Name=Dyn_Tail_Master, swing1LimitDegrees=15, swing2LimitDegrees=10, twistLimitDegrees=5}
 {bone1Name=Dyn_Tail_2, bone2Name=Dyn_Tail_3, swing1LimitDegrees=30, swing2LimitDegrees=30, twistLimitDegrees=5}
 {bone1Name=Dyn_Tail_3, bone2Name=Dyn_Tail_1, swing1LimitDegrees=35, swing2LimitDegrees=10, twistLimitDegrees=5}
]
