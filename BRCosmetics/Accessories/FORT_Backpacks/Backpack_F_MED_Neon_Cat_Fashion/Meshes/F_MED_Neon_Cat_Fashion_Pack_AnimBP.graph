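# Animation Blueprint for the Neon Cat Fashion backpack, adding physics motion and flight-state adjustments to its pack and tail bones.
# A fixed additive offset on Dyn_Tail_Master feeds a RigidBody simulation in pack-bone space.
# Skydive floating, open parachute and skydiving flags then blend in additive pack offsets, followed by a second, differently tuned RigidBody pass.
# The dive-mode offset node is unconnected, and the EventGraph has no logic.
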AnimBlueprint /BRCosmetics/Accessories/FORT_Backpacks/Backpack_F_MED_Neon_Cat_Fashion/Meshes/F_MED_Neon_Cat_Fashion_Pack_AnimBP
target_skeleton=&/F_MED_Neon_Cat_Fashion_Pack_Skeleton
graphs [
 AnimGraph {nodes [
   "Transform (Modify) Bone - Bone: pack"!(ComponentPose, bAlphaBoolEnabled:bool=Get bIsSkydiveDiveMode(), Translation="0,-2,0", Rotation="0,0,20", Scale) -> (Pose) {
    alphaInputType=Bool
    rotationMode=BMM_Additive
    rotationSpace=BCS_BoneSpace
    translationMode=BMM_Additive
    translationSpace=BCS_BoneSpace
    alphaBoolBlend {blendInTime=10, blendOption=ExpInOut, blendOutTime=10}
    boneToModify {boneName=pack}
    rotation {roll=20}
    translation {y=-2}
   }
   n1 = "Transform (Modify) Bone - Bone: Dyn_Tail_Master"(ComponentPose=Local To Component(LocalPose=Input Pose()), Alpha=1, Translation="0,-0.5,0", Rotation="0,-5,0", Scale="1,1,1") {
    rotationMode=BMM_Additive
    rotationSpace=BCS_BoneSpace
    translationMode=BMM_Additive
    translationSpace=BCS_BoneSpace
    alphaBoolBlend {blendOption=Linear}
    boneToModify {boneName=Dyn_Tail_Master}
    rotation {yaw=-5}
    translation {y=-0.5}
   }
   n2 = RigidBody(ComponentPose=@n1) {
    bClampLinearTranslationLimitToRefPose=true
    bFreezeIncomingPoseOnStart=true
    bOverrideWorldGravity=true
    overridePhysicsAsset=&/F_MED_Neon_Cat_Fashion_Pack_Physics
    simulationSpace=BaseBoneSpace
    alphaBoolBlend {blendOption=Linear}
    baseBoneRef {boneName=pack}
    componentAppliedLinearAccClamp=<5000 5000 5000>
    componentLinearAccScale=<0.1 0.2 0.1>
    componentLinearVelScale=<0.1 0.1 0.1>
    overrideWorldGravity {z=-400}
   }
   n3 = "Transform (Modify) Bone - Bone: pack"(ComponentPose=@n2, bAlphaBoolEnabled=Get bIsSkydiveFloating(), Translation="0,-2,0", Rotation="0,0,20") {
    alphaInputType=Bool
    rotationMode=BMM_Additive
    rotationSpace=BCS_BoneSpace
    translationMode=BMM_Additive
    translationSpace=BCS_BoneSpace
    alphaBoolBlend {blendInTime=1, blendOption=Linear, blendOutTime=1}
    boneToModify {boneName=pack}
    rotation {roll=20}
    translation {y=-2}
   }
   n4 = "Transform (Modify) Bone - Bone: pack"(ComponentPose=@n3, bAlphaBoolEnabled=Get bIsParachuteOpened(), Translation="0,-2,0", Rotation="0,0,20") {
    alphaInputType=Bool
    rotationMode=BMM_Additive
    rotationSpace=BCS_BoneSpace
    translationMode=BMM_Additive
    translationSpace=BCS_BoneSpace
    alphaBoolBlend {blendInTime=1, blendOption=Linear, blendOutTime=5}
    boneToModify {boneName=pack}
    rotation {roll=20}
    translation {y=-2}
   }
   n5 = "Transform (Modify) Bone - Bone: pack"(ComponentPose=@n4, bAlphaBoolEnabled=Get bIsSkyDiving(), Translation="0,-2,0", Rotation="0,0,20") {
    alphaInputType=Bool
    rotationMode=BMM_Additive
    rotationSpace=BCS_BoneSpace
    translationMode=BMM_Additive
    translationSpace=BCS_BoneSpace
    alphaBoolBlend {blendInTime=1, blendOption=Linear, blendOutTime=1}
    boneToModify {boneName=pack}
    rotation {roll=20}
    translation {y=-2}
   }
   n6 = RigidBody(ComponentPose=@n5) {
    bClampLinearTranslationLimitToRefPose=true
    bFreezeIncomingPoseOnStart=true
    bOverrideWorldGravity=true
    overridePhysicsAsset=&F_MED_Neon_Cat_Fashion_Pack_Physics
    simulationSpace=BaseBoneSpace
    alphaBoolBlend {blendOption=Linear}
    baseBoneRef {boneName=pack}
    componentAppliedLinearAccClamp=<1000 1000 1000>
    componentLinearAccScale=<1 0.2 0.1>
    componentLinearVelScale=<1 0.1 0.1>
    overrideWorldGravity {z=-200}
   }
   Output Pose(Result=Component To Local(ComponentPose=@n6)) {layerGroup=None}
  ]}
 EventGraph {nodes [
   TryGetPawnOwner()
   Event BlueprintUpdateAnimation()
  ]}
]
variables []
